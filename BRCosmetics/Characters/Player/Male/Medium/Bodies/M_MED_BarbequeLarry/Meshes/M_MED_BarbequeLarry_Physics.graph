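# Physics asset for the M_MED_BarbequeLarry male body cosmetic, covering dangling waist gear: a belt, a knife and two grenades.
# A pelvis capsule is the root body, with the belt box constrained directly to it by tight angular limits.
# The knife and each grenade are capsules with limited swing and twist against their own tiny sphere parent body.
# Those parent spheres connect to the pelvis with free swing and twist.
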PhysicsAsset /BRCosmetics/Characters/Player/Male/Medium/Bodies/M_MED_BarbequeLarry/Meshes/M_MED_BarbequeLarry_Physics
bodies {
 dyn_belt [{shapeType=Box, extentX=5.2794, extentY=1, extentZ=5, center=<2.0288 -0.3118 0>, rotation=<rot 0 -8.8023 0>}]
 dyn_grenadeA_01 [{shapeType=Capsule, length=7.0576, radius=4.1554, center=<1.7284 4.2194 0.1906>, rotation=<rot -0.2032 -83.4472 -87.2783>}]
 dyn_grenadeB_01 [{shapeType=Capsule, length=7.0576, radius=4.6504, center=<3.2677 4.1759 0.2616>, rotation=<rot 0 -86.9359 -89.7333>}]
 dyn_knife_01 [{shapeType=Capsule, length=22.7368, radius=4.731, center=<12.1283 0.1652 1.3123>, rotation=<rot 0 -87.4189 -88.4308>}]
 dyn_parent_grenadeA [{shapeType=Sphere, radius=0.505}]
 dyn_parent_grenadeB [{shapeType=Sphere, radius=0.505}]
 dyn_parent_knife [{shapeType=Sphere, radius=0.505}]
 pelvis [{shapeType=Capsule, length=7.5, radius=15.25, center=<-4.0862 3.1249 0>}]
}
constraints [
 {bone1Name=dyn_knife_01, bone2Name=dyn_parent_knife, swing1LimitDegrees=10, swing2LimitDegrees=45, twistLimitDegrees=5}
 {bone1Name=dyn_grenadeA_01, bone2Name=dyn_parent_grenadeA, swing1LimitDegrees=45, swing2LimitDegrees=45, twistLimitDegrees=45}
 {
  bone1Name=dyn_parent_knife
  bone2Name=pelvis
  swing1LimitDegrees=45
  swing1Motion=Free
  swing2LimitDegrees=45
  swing2Motion=Free
  twistLimitDegrees=45
  twistMotion=Free
 }
 {
  bone1Name=dyn_parent_grenadeA
  bone2Name=pelvis
  swing1LimitDegrees=45
  swing1Motion=Free
  swing2LimitDegrees=45
  swing2Motion=Free
  twistLimitDegrees=45
  twistMotion=Free
 }
 {
  bone1Name=dyn_parent_grenadeB
  bone2Name=pelvis
  swing1LimitDegrees=45
  swing1Motion=Free
  swing2LimitDegrees=45
  swing2Motion=Free
  twistLimitDegrees=45
  twistMotion=Free
 }
 {bone1Name=dyn_grenadeB_01, bone2Name=dyn_parent_grenadeB, swing1LimitDegrees=45, swing2LimitDegrees=45, twistLimitDegrees=20}
 {bone1Name=dyn_belt, bone2Name=pelvis, swing1LimitDegrees=15, swing2LimitDegrees=15, twistLimitDegrees=10}
]
