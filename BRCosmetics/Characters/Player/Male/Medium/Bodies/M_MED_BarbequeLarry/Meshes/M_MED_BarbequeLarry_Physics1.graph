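# Physics asset for the BarbequeLarry medium male player body, covering two dangling dynamic chains, top and bottom, at the right thigh.
# Each chain's small sphere root body is constrained to the thigh_r capsule with free swing and twist.
# Each root links to a box-shaped second body with 60-degree swing and twist limits.
# The bottom chain's second body combines two boxes, while the top uses one.
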PhysicsAsset /BRCosmetics/Characters/Player/Male/Medium/Bodies/M_MED_BarbequeLarry/Meshes/M_MED_BarbequeLarry_Physics1
bodies {
 dyn_chain_bot_mid_01 [{shapeType=Sphere, radius=1}]
 dyn_chain_bot_mid_02 [
  {shapeType=Box, extentX=3.0598, extentY=16.8938, extentZ=25.8529, center=<-0.8141 -14.5098 0.6455>, rotation=<rot -1.2863 -94.9475 -3.187>}
  {shapeType=Box, extentX=19, extentY=24.5, extentZ=4.5, center=<0.0001 -6.0482 -13.6076>, rotation=<rot 0.2585 -6.1589 2.3945>}
 ]
 dyn_chain_top_mid_01 [{shapeType=Sphere, radius=0.505}]
 dyn_chain_top_mid_02 [{shapeType=Box, extentX=3.573, extentY=16.0619, extentZ=23.9381, center=<-0.6616 -14.2308 -0.3206>, rotation=<rot -1.146 -96.7784 -3.3404>}]
 thigh_r [{shapeType=Capsule, length=25.75, radius=7.75, center=<18.8264 0.0006 0>, rotation=<rot 90 0 0>}]
}
constraints [
 {
  bone1Name=dyn_chain_bot_mid_01
  bone2Name=thigh_r
  swing1LimitDegrees=45
  swing1Motion=Free
  swing2LimitDegrees=45
  swing2Motion=Free
  twistLimitDegrees=45
  twistMotion=Free
 }
 {
  bone1Name=dyn_chain_top_mid_01
  bone2Name=thigh_r
  swing1LimitDegrees=45
  swing1Motion=Free
  swing2LimitDegrees=45
  swing2Motion=Free
  twistLimitDegrees=45
  twistMotion=Free
 }
 {bone1Name=dyn_chain_bot_mid_02, bone2Name=dyn_chain_bot_mid_01, swing1LimitDegrees=60, swing2LimitDegrees=60, twistLimitDegrees=60}
 {bone1Name=dyn_chain_top_mid_02, bone2Name=dyn_chain_top_mid_01, swing1LimitDegrees=60, swing2LimitDegrees=60, twistLimitDegrees=60}
]
