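# Post-style secondary-motion AnimBP for the BarbequeLarry body cosmetic, copying its pose from the attached parent mesh.
# A weighted constraint ties dyn_clip_aim to right thigh bones, then a Control Rig runs.
# AnimDynamics on clip, chain and grenade-ring bones interleave with two RigidBody passes using separate physics assets before output.
# Look At, two Flap and second-segment chain AnimDynamics nodes are unconnected; the event graph holds no logic.
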AnimBlueprint /BRCosmetics/Characters/Player/Male/Medium/Bodies/M_MED_BarbequeLarry/Meshes/M_MED_BarbequeLarry_AnimBP
target_skeleton=&/M_MED_BarbequeLarry_Skeleton
graphs [
 AnimGraph {nodes [
   Anim Dynamics - Bone: dyn_chain_top_mid_02(Alpha=1, ExternalForce="0,0,0") {
    angularDampingOverride=0.9999
    angularSpringConstant=1500
    bAngularSpring=true
    bGravityOverrideInSimSpace=true
    bOverrideAngularDamping=true
    gravityScale=0.4
    simulationSpace=RootRelative
    alphaBoolBlend {blendOption=Linear}
    boundBone {boneName=dyn_chain_top_mid_02}
    componentAppliedLinearAccClamp=<1300 1300 900>
    componentLinearAccScale=<0.8 0.8 0.6>
    componentLinearVelScale=<0.8 0.8 0.6>
    physicsBodyDefinitions [{
      sphereCollisionRadius=0
      boundBone {boneName=dyn_chain_top_mid_02}
      boxExtents=<5 25 20>
      constraintSetup {angularLimitsMax=<15 1 3>, angularLimitsMin=<-15 -1 -15>, angularTarget=<0 0 0>}
      localJointOffset=<5 0 0>
     }]
   }
   Anim Dynamics - Bone: dyn_chain_bot_mid_02(Alpha=0.7, ExternalForce="0,0,0") {
    angularDampingOverride=1
    angularSpringConstant=1500
    bAngularSpring=true
    bGravityOverrideInSimSpace=true
    bOverrideAngularDamping=true
    gravityScale=0.35
    simulationSpace=RootRelative
    alphaBoolBlend {blendOption=Linear}
    boundBone {boneName=dyn_chain_bot_mid_02}
    componentAppliedLinearAccClamp=<1300 1300 900>
    componentLinearAccScale=<0.8 0.8 0.6>
    componentLinearVelScale=<0.8 0.8 0.6>
    physicsBodyDefinitions [{
      sphereCollisionRadius=0
      boundBone {boneName=dyn_chain_bot_mid_02}
      boxExtents=<10 25 25>
      constraintSetup {angularLimitsMax=<15 1 3>, angularLimitsMin=<-15 -1 -15>, angularTarget=<0 0 0>}
      localJointOffset=<8 0 0>
     }]
   }
   Look At - Bone: dyn_clip_01(Alpha=1) {
    alphaBoolBlend {blendOption=Linear}
    boneToModify {boneName=dyn_clip_01}
    lookAtLocation=<6 4>
    lookAtTarget {boneReference {boneName=dyn_clip_aim}}
    lookAt_Axis {axis=<1 0>}
    lookUp_Axis {axis {x=0}}
   }
   n1 = "Constraint - dyn_clip_aim (PP)"!(
    ComponentPose=Local To Component(LocalPose=Copy Pose From Mesh() {bUseAttachedParent=true})
    Alpha=1
    ConstraintWeights_0=0.1141
    ConstraintWeights_1:float=0.8859
   ) -> (Pose) {
    alphaBoolBlend {blendOption=Linear}
    boneToModify {boneName=dyn_clip_aim}
    constraintSetup [{targetBone {boneName=thigh_r}}, {targetBone {boneName=thigh_twist_01_r}}]
    constraintWeights [1,1]
   }
   n2 = Control Rig(Source=Component To Local(ComponentPose=@n1)) {
    alphaBoolBlend {blendOption=Linear}
    controlRigAssetReference {blueprintRigClass=&/M_MED_BarbequeLarry_CtrlRig.M_MED_BarbequeLarry_CtrlRig_C}
   }
   n3 = Anim Dynamics - Bone: dyn_clip_01(ComponentPose=Local To Component(LocalPose=@n2), Alpha=0.8, ExternalForce="0,0,0") {
    angularDampingOverride=0.9999
    angularSpringConstant=1500
    bAngularSpring=true
    bGravityOverrideInSimSpace=true
    bOverrideAngularDamping=true
    gravityScale=0.35
    simulationSpace=RootRelative
    alphaBoolBlend {blendOption=Linear}
    boundBone {boneName=dyn_clip_01}
    physicsBodyDefinitions [{
      sphereCollisionRadius=0
      boundBone {boneName=dyn_clip_01}
      boxExtents=<10 30 20>
      constraintSetup {angularLimitsMax=<5 35 1>, angularLimitsMin=<-5 -2 -40>, angularTarget=<0 0 0>}
      localJointOffset=<10 0 0>
     }]
   }
   n4 = Anim Dynamics - Bone: dyn_chain_top_mid_01(ComponentPose=@n3, Alpha=0.6, ExternalForce="0,0,0") {
    angularDampingOverride=0.9999
    angularSpringConstant=1500
    bAngularSpring=true
    bGravityOverrideInSimSpace=true
    bOverrideAngularDamping=true
    gravityScale=0.355
    simulationSpace=RootRelative
    alphaBoolBlend {blendOption=Linear}
    boundBone {boneName=dyn_chain_top_mid_01}
    componentAppliedLinearAccClamp=<1200 1200 1000>
    componentLinearAccScale=<0.8 0.8 0.2>
    componentLinearVelScale=<0.8 0.8 0.2>
    physicsBodyDefinitions [{
      sphereCollisionRadius=0
      boundBone {boneName=dyn_chain_top_mid_01}
      boxExtents=<5 25 10>
      constraintSetup {angularLimitsMax=<15 1 0>, angularLimitsMin=<-15 -1 -20>, angularTarget=<0 0 0>}
      localJointOffset=<5 0 0>
     }]
   }
   n5 = Anim Dynamics - Bone: dyn_chain_bot_mid_01(ComponentPose=@n4, Alpha=0.6, ExternalForce="0,0,0") {
    angularDampingOverride=1
    angularSpringConstant=1500
    bAngularSpring=true
    bGravityOverrideInSimSpace=true
    bOverrideAngularDamping=true
    gravityScale=0.31
    simulationSpace=RootRelative
    alphaBoolBlend {blendOption=Linear}
    boundBone {boneName=dyn_chain_bot_mid_01}
    componentAppliedLinearAccClamp=<1200 1200 1000>
    componentLinearAccScale=<0.8 0.8 0.2>
    componentLinearVelScale=<0.8 0.8 0.2>
    physicsBodyDefinitions [{
      sphereCollisionRadius=0
      boundBone {boneName=dyn_chain_bot_mid_01}
      boxExtents=<10 25 25>
      constraintSetup {angularLimitsMax=<15 1 0>, angularLimitsMin=<-15 -1 -20>, angularTarget=<0 0 0>}
      localJointOffset=<8 0 0>
     }]
   }
   n6 = RigidBody(ComponentPose=@n5, Alpha=0.8) {
    bClampLinearTranslationLimitToRefPose=true
    bOverrideWorldGravity=true
    overridePhysicsAsset=&/M_MED_BarbequeLarry_Physics1
    simulationSpace=BaseBoneSpace
    alphaBoolBlend {blendOption=Linear}
    baseBoneRef {boneName=root}
    componentAppliedLinearAccClamp=<1100 1100 950>
    componentLinearAccScale=<0.6 0.6 0.4>
    componentLinearVelScale=<0.6 0.6 0.4>
    overrideWorldGravity {z=-900}
   }
   n7 = Anim Dynamics - Bone: dyn_grenadeA_ring_01(ComponentPose=@n6, Alpha=1, ExternalForce="0,0,0") {
    angularDampingOverride=0.9999
    angularSpringConstant=1500
    bAngularSpring=true
    bGravityOverrideInSimSpace=true
    bOverrideAngularDamping=true
    gravityScale=0.35
    simulationSpace=RootRelative
    alphaBoolBlend {blendOption=Linear}
    boundBone {boneName=dyn_grenadeA_ring_01}
    componentAppliedLinearAccClamp=<1100 1100 1200>
    componentLinearAccScale=<0.7 0.7 0.5>
    componentLinearVelScale=<0.7 0.7 0.5>
    physicsBodyDefinitions [{
      sphereCollisionRadius=0
      boundBone {boneName=dyn_grenadeA_ring_01}
      boxExtents=<5 15 15>
      constraintSetup {angularLimitsMax=<5 20 3>, angularLimitsMin=<-5 -20 -40>, angularTarget=<0 0 0>}
      localJointOffset=<5 0 1>
     }]
   }
   n8 = Anim Dynamics - Bone: dyn_grenadeB_ring_01(ComponentPose=@n7, Alpha=1, ExternalForce="0,0,0") {
    angularDampingOverride=0.9999
    angularSpringConstant=1500
    bAngularSpring=true
    bGravityOverrideInSimSpace=true
    bOverrideAngularDamping=true
    gravityScale=0.35
    simulationSpace=RootRelative
    alphaBoolBlend {blendOption=Linear}
    boundBone {boneName=dyn_grenadeB_ring_01}
    componentAppliedLinearAccClamp=<1100 1100 1200>
    componentLinearAccScale=<0.7 0.7 0.4>
    componentLinearVelScale=<0.7 0.7 0.4>
    physicsBodyDefinitions [{
      sphereCollisionRadius=0
      boundBone {boneName=dyn_grenadeB_ring_01}
      boxExtents=<5 16 12>
      constraintSetup {angularLimitsMax=<5 20 3>, angularLimitsMin=<-5 -20 -40>, angularTarget=<0 0 0>}
      localJointOffset=<4.5 0 -0.5>
     }]
   }
   n9 = RigidBody(ComponentPose=@n8) {
    bClampLinearTranslationLimitToRefPose=true
    bOverrideWorldGravity=true
    overridePhysicsAsset=&/M_MED_BarbequeLarry_Physics
    simulationSpace=BaseBoneSpace
    alphaBoolBlend {blendOption=Linear}
    baseBoneRef {boneName=root}
    componentAppliedLinearAccClamp=<1100 1100 1000>
    componentLinearAccScale=<0.6 0.6 0.5>
    componentLinearVelScale=<0.6 0.6 0.5>
    overrideWorldGravity {z=-900}
   }
   Output Pose(Result=Component To Local(ComponentPose=@n9)) {layerGroup=None}
   Flap Node - Bone: dyn_parent_clip!(ComponentPose:Component Space Pose Link Structure, Alpha:float=1) -> (Pose:Component Space Pose Link Structure) {
    is=title FortAnimGraphNode_Flap
    bUseClamp=true
    targetClampMax=10
    targetClampMin=-10
    targetComponent=RotationY
    alphaBoolBlend {blendOption=Linear}
    sourceDrives [
     {
      bUseRange=false
      drivingCurve=null
      multiplier=1
      rangeMax=1
      rangeMin=-1
      remappedMax=1
      remappedMin=0
      sourceComponent=RotationY
      useQuaternion=true
      sourceBone {boneName=thigh_twist_01_r}
     }
     {
      bUseRange=false
      drivingCurve=null
      multiplier=0.5
      rangeMax=1
      rangeMin=-1
      remappedMax=1
      remappedMin=0
      sourceComponent=RotationZ
      useQuaternion=true
      sourceBone {boneName=thigh_r}
     }
    ]
    targetBone {boneName=dyn_parent_clip}
   }
   Flap Node - Bone: dyn_parent_clip(Alpha=1) {
    bUseClamp=true
    targetClampMax=10
    targetClampMin=-10
    targetComponent=RotationZ
    alphaBoolBlend {blendOption=Linear}
    sourceDrives [
     {
      bUseRange=false
      drivingCurve=null
      multiplier=-0.5
      rangeMax=1
      rangeMin=-1
      remappedMax=1
      remappedMin=0
      sourceComponent=RotationX
      useQuaternion=true
      sourceBone {boneName=thigh_r}
     }
     {
      bUseRange=false
      drivingCurve=null
      multiplier=0.5
      rangeMax=1
      rangeMin=-1
      remappedMax=1
      remappedMin=0
      sourceComponent=RotationZ
      useQuaternion=true
      sourceBone {boneName=thigh_r}
     }
    ]
    targetBone {boneName=dyn_parent_clip}
   }
  ]}
 EventGraph {nodes [Event BlueprintUpdateAnimation()]}
]
variables []
